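# Niagara system for the IconColors drift trail vehicle cosmetic, adding ribbon, mesh and sprite effects at the wheels.
# Two wavy ribbons stream backward; burst-spawned mesh/sprite particles and rotating wheel-wrap meshes sample skeletal mesh sockets, with a second wheel-wrap emitter disabled.
# Painted-variant overrides color each emitter, and ColorVariantID arrays tune contrast and desaturation on ribbons and wraps.
# Shared modules unpack vehicle flags, mask per-tire effects and handle product previews.
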NiagaraSystem /VehicleCosmetics/DriftTrail/DriftTrail_IconColors/FX/NS_DriftTrail_IconColors
dependencies {
 usedDataInterfaces [&NiagaraDataInterfaceCurve, &NiagaraDataInterfaceArrayFloat3, &NiagaraDataInterfaceSkeletalMesh]
 usedDynamicInputs []
 usedModules [
  &self:SystemSpawnScript.SystemScriptSource.SystemScriptGraph.NiagaraNodeAssignment_0.SetVariables_#1
  &self:SystemSpawnScript.SystemScriptSource.SystemScriptGraph.NiagaraNodeAssignment_1.SetVariables_#2
  &self:SystemSpawnScript.SystemScriptSource.SystemScriptGraph.NiagaraNodeAssignment_2.SetVariables_#3
  &FN_UnpackVehicleFlags
  /VehicleCosmetics/FX/Niagara/Modules/VC_SetVehicleSpeedAttribs
  &SystemState
 ]
}
emitters {
 NE_VC_DriftTrail_Ribbon_Parent {
  emitterSpawnScript [/SetVariables(
    Emitter.ContrastTx1:NiagaraFloat=&SelectFloatFromArray(
     Direct Array Index=$System.ColorVariantID
     Float Selection Array=DataInterface() {propertyValues {FloatData [0,0,1.25,1,1.25,0.75,1.25,0.95,1,1,0.97,1.25,1.25,0.95]}}
    )
    Emitter.Desaturation:NiagaraFloat=&SelectFloatFromArray(Direct Array Index=$System.ColorVariantID, Float Selection Array=DataInterface() {propertyValues {FloatData [1,0,0,0,0,0,0,0,0,0,0,0,0,0]}})
   ) {isSetParameters=true}]
  emitterUpdateScript [
   &EmitterState()
   &VC_RateQualityHandler(Default Rate:NiagaraFloat=&Multiply_Float(
     A=100
     B=/VehicleCosmetics/FX/Niagara/Modules/VC_Mask_Floats(Coro Value:NiagaraFloat=1, FrontEnd Value:NiagaraFloat=0.75, FN Value:NiagaraFloat=0.5)
    ))
   &VC_SpawnRibbon_v1(Number of Ribbons=2, SpawnRate=$Emitter.Spawn Rate)
  ]
  particleSpawnScript [
   &VC_PaintedVariantsOverride(
    Black=<rgba 0.2 0.2 0.2 8>
    BurntSienna=<rgba 0.68 0.3373 0.2992 4>
    Cobalt=<rgba 0 0.075 0.9 4>
    Crimson=<rgba 1 0.1 0.1244 4>
    ForestGreen=<rgba 0.125 0.5 0.1439 8>
    Grey=<rgba 0.432 0.4387 0.48 4>
    Lime=<rgba 0.533 0.82 0 4>
    Orange=<rgba 1.12 0.28 0 4>
    Pink=<rgba 1.1 0.275 0.9625 4>
    Purple=<rgba 0.4934 0.28 0.8 8>
    Saffron=<rgba 1.1 1.0688 0.165 8>
    SkyBlue=<rgba 0 0.74 1.2 4>
    TitaniumWhite=<rgba 1.1 1.1 1.1 4>
   )
   &VC_InitializeRibbon_v1(
    Color=$StackContext.Color
    Custom Offset=true
    Enable Skeletal Position=true
    Lifetime=0.95
    Local Vector Offset=&AddVectorToPosition(Position=<-55 0 -30>, Vector=$User.offset)
    Offset Amount=1
   )
   &AddVelocity(Velocity=<-1000 0 0>) {module=AddVelocity001}
   /VehicleCosmetics/FX/VC_FXParents/Modules/DLR_ProductPreviews(GameModeID:NiagaraInt32=$System.GameModeID, bPreviewVehicleSpeed:NiagaraBool=$System.bPreviewVehicleSpeed)
   /SetVariables(
    Particles.R1Width:NiagaraFloat=&SelectFloatFromArray(Direct Array Index=&SpawnGroup(), Float Selection Array=DataInterface() {propertyValues {FloatData [0,-50]}})
    Particles.R2Width:NiagaraFloat=&SelectFloatFromArray(Direct Array Index=&SpawnGroup(), Float Selection Array=DataInterface() {propertyValues {FloatData [50,0]}})
   ) {isSetParameters=true}
   /SetVariables() {isSetParameters=true}
  ]
  particleUpdateScript [
   &FN_DisableTireFX(Mask Index Attribute:NiagaraInt32=$Particles.SpawnGroup, Mask Index FL=2, Mask Index FR=3, Mask Index RL=0, Mask Index RR=1)
   &ParticleState()
   &InheritSourceMovement(
    Applied Position Scale=&Multiply_VectorByFloat(
     Vector=<1 1 1>
     Float=&FloatFromCurve(FloatCurve=DataInterface() {propertyValues {Curve {
         defaultValue=1
         keys [
          {arriveTangent=-3.9073, interpMode=RCIM_Cubic, leaveTangent=-3.9073, tangentMode=RCTM_User, time=0, value=1}
          {tangentMode=RCTM_User, time=1, value=0}
         ]
        }}})
    )
    ApplyForce=false
    ApplyPosition=true
    ApplyPositionFalloff=false
   )
   &SolveForcesAndVelocity(Clamp Velocity=false, Limit Acceleration=false)
   &DynamicMaterialParameters(
    Index 0 Param 1=&FloatFromCurve(FloatCurve=DataInterface() {propertyValues {Curve {keys [<at 0.6 0>,<at 0.9 0.05>]}}}, Scale Curve=10)
    Module.Index0Float0Write=true
    Module.Index0Float1Write=true
    Module.Index0Float2Write=true
    Module.Index0Float3Write=true
   )
   &VC_UpdateRibbon_v1(Enable Skeletal Position=true)
   /SetVariables(Particles.CustomPosition_R1:NiagaraPosition=&AddVectorToPosition(
     Position=$Particles.Position
     Vector=&MakeVector(
      X=0
      Y=0
      Z=&MaskFloat(
       Mask Source=0
       Mask Parameter=?
       Mask=0
       Masked Float=&Cosine(
        Normalized Angle=&Add_Float(A=&Multiply_Float(A=$Emitter.LoopedAge, B=1), B=$Particles.NormalizedAge)
        Period=1
        Scale=&FloatFromCurve(
         FloatCurve=DataInterface() {propertyValues {Curve {
            defaultValue=0.3
            keys [{interpMode=RCIM_Cubic, tangentMode=RCTM_User, time=0.3, value=0}, {interpMode=RCIM_Cubic, tangentMode=RCTM_User, time=0.5, value=1}]
           }}}
         Scale Curve=1.5
        )
        Bias=0
       )
       Passthrough Float=&Sine(
        Normalized Angle=&Add_Float(A=&Multiply_Float(A=$Emitter.LoopedAge, B=1), B=$Particles.NormalizedAge)
        Period=1
        Scale=&FloatFromCurve(
         FloatCurve=DataInterface() {propertyValues {Curve {
            defaultValue=0.3
            keys [{interpMode=RCIM_Cubic, tangentMode=RCTM_User, time=0.3, value=0}, {interpMode=RCIM_Cubic, tangentMode=RCTM_User, time=0.5, value=1}]
           }}}
         Scale Curve=1.5
        )
        Bias=0
       )
      )
     )
    )) {isSetParameters=true}
   /SetVariables(
    Particles.R1BlueWidth:NiagaraFloat=&Multiply_Float(
     A=$Particles.R1Width
     B=&FloatFromCurve(
      FloatCurve=DataInterface() {propertyValues {Curve {
         defaultValue=1
         keys [
          {arriveTangent=12.0019, interpMode=RCIM_Cubic, leaveTangent=12.0019, tangentMode=RCTM_User, time=0.02, value=0}
          {arriveTangent=0.9967, interpMode=RCIM_Cubic, leaveTangent=0.9967, tangentMode=RCTM_User, time=0.2, value=1.5}
          {tangentMode=RCTM_User, time=1, value=0.25}
         ]
        }}}
      Scale Curve=0.5
     )
    )
    Particles.R2BlueWidth:NiagaraFloat=&Multiply_Float(
     A=$Particles.R2Width
     B=&FloatFromCurve(
      FloatCurve=DataInterface() {propertyValues {Curve {
         defaultValue=1
         keys [
          {arriveTangent=14.3715, interpMode=RCIM_Cubic, leaveTangent=14.3715, tangentMode=RCTM_User, time=0.02, value=0}
          {arriveTangent=0.2432, interpMode=RCIM_Cubic, leaveTangent=0.2432, tangentMode=RCTM_User, time=0.2, value=1.5}
          {tangentMode=RCTM_User, time=1, value=0.25}
         ]
        }}}
      Scale Curve=0.5
     )
    )
   ) {isSetParameters=true}
  ]
  rendererClasses [&NiagaraRibbonRendererProperties]
  renderers [
   {rendererClass=&NiagaraRibbonRendererProperties}
   {rendererClass=&NiagaraRibbonRendererProperties, rendererIndex=1}
  ]
 }
 NE_VC_DriftTrail_Sprite_Parent {
  emitterUpdateScript [
   &EmitterState(
    Inactive Response=::NewEnumerator1
    Life Cycle Mode=::NewEnumerator1
    Loop Behavior=::NewEnumerator1
    Loop Duration Mode=::NewEnumerator1
    UseLoopDelay=false
   )
   &SpawnBurst_Instantaneous(Spawn Count=4, Use Spawn Probability=false, UseLoopCountLimit=false)
   &SpawnRate(
    SpawnRate=&VC_Mask_Floats(
     FrontEnd Value:NiagaraFloat=200
     blsFrontEndPreview:NiagaraBool=false
     SpeedMin Value:NiagaraFloat=120
     SpeedMax Value:NiagaraFloat=180
     Speed Alpha:NiagaraFloat=$System.Speed_Alpha
    )
    Use Spawn Probability=false
   ) {enabled=false}
  ]
  particleSpawnScript [
   &VC_PaintedVariantsOverride(
    Black=<rgba 0.2 0.2 0.2 4>
    BurntSienna=<rgba 0.65 0.299 0.26 4>
    Cobalt=<rgba 0 0.075 0.9 4>
    Crimson=<rgba 1 0.1 0.1244 4>
    DefaultColor=<rgba 1.35 0.2742 1.19 1>
    ForestGreen=<rgba 0.125 0.5 0.1439 4>
    Grey=<rgba 0.432 0.4387 0.48 4>
    Lime=<rgba 0.533 0.82 0 4>
    Orange=<rgba 1.12 0.28 0 4>
    Pink=<rgba 1.1 0.275 0.9625 4>
    Purple=<rgba 0.4934 0.28 0.8 4>
    R1 Black:LinearColor=<rgba 0.35 0.35 0.35 4>
    R1 BurntSienna:LinearColor=<rgba 0.8 0.3967 0.36 4>
    R1 Cobalt:LinearColor=<rgba 0.036 0.327 1.2 4>
    R1 Crimson:LinearColor=<rgba 1.4 0.1254 0.07 4>
    R1 DefaultColor=<rgba 0 0.9975 2 1>
    R1 ForestGreen:LinearColor=<rgba 0 0.7 0.0354 4>
    R1 Grey:LinearColor=<rgba 0.7 0.7 0.7 4>
    R1 Lime:LinearColor=<rgba 0.675 0.9 0 4>
    R1 Orange:LinearColor=<rgba 1.4 0.35 0 4>
    R1 Pink:LinearColor=<rgba 1.3 0.325 1.1375 4>
    R1 Purple:LinearColor=<rgba 0.6168 0.35 1 4>
    R1 Saffron:LinearColor=<rgba 1.4 1.3603 0.21 4>
    R1 SkyBlue:LinearColor=<rgba 0 0.8167 1.4 4>
    R1 TitaniumWhite:LinearColor=<rgba 1.1 1.1 1.1 4>
    Saffron=<rgba 1.1 1.0688 0.165 4>
    SkyBlue=<rgba 0 0.74 1.2 4>
    TitaniumWhite=<rgba 1.1 1.1 1.1 4>
    Use R1 Painted Overrides=true
   )
   &InitializeParticle(
    Color Mode=::NewEnumerator0
    Mesh Scale=<4 1 1>
    Mesh Scale Mode=::NewEnumerator4
    Sprite Rotation Mode=::NewEnumerator2
    Sprite Size Mode=::NewEnumerator3
    UsePositionOffset=false
    Write Material Random=true
   )
   &PartitionParticles()
   &SkeletalMeshLocation(
    Bone / Socket Index=&Add_Integer(A=$Particles.PartitionParticles.Partition, B=4)
    Bone / Socket Sampling Mode=::NewEnumerator5
    MaskSpawnGroup=false
    Mesh Sampling Type=::NewEnumerator6
    Sampled Position Offset=&AddVectorToPosition(Position=<-51 0 -30>, Vector=$User.offset)
    Skeletal Mesh=$System.SkeletalMesh
   )
   &AddVelocity(Velocity=<0 0 0>) {enabled=false}
   &InitialMeshOrientation(Mesh Axis To Orient=<-1 0 0>, Mesh Orientation Mode=::NewEnumerator2, Orientation Vector=<0 0 0>, Use Rotation Vector=true)
   &DLR_ProductPreviews(GameModeID:NiagaraInt32=$System.GameModeID, bPreviewVehicleSpeed:NiagaraBool=$System.bPreviewVehicleSpeed)
  ]
  particleUpdateScript [
   &FN_DisableTireFX(
    Mask Index Attribute:NiagaraInt32=$Particles.PartitionParticles.Partition
    Mask Index FL=2
    Mask Index FR=3
    Mask Index RL=0
    Mask Index RR=1
   )
   &ParticleState(Kill Particles When Lifetime Has Elapsed=false, Loop Particles Lifetime=true)
   &InheritSourceMovement(
    Applied Position Falloff Distance=650
    Applied Position Scale=&VectorFromFloat(Value=&FloatFromCurve(FloatCurve=DataInterface() {propertyValues {Curve {
         defaultValue=1
         keys [
          {arriveTangent=-1.7885, interpMode=RCIM_Cubic, leaveTangent=-1.7885, tangentMode=RCTM_User, time=0.05, value=1}
          {interpMode=RCIM_Cubic, tangentMode=RCTM_User, time=1, value=0}
         ]
        }}}))
    ApplyForce=false
    ApplyPosition=true
    ApplyPositionFalloff=false
   ) {enabled=false, module=InheritSourceMovement001}
   /SetVariables(Particles.BlueCone:LinearColor=<rgba 0 0.9975 2 1>, Particles.BlueConeSize:Vector3f=<4 0.8 0.8>, Particles.TipOffset=0.1) {isSetParameters=true}
   &DynamicMaterialParameters(
    Index 0 Param 1=&FloatFromCurve(
     FloatCurve=DataInterface() {propertyValues {Curve {
        defaultValue=1
        keys [{interpMode=RCIM_Cubic, tangentMode=RCTM_User, time=0.5, value=1}, {interpMode=RCIM_Cubic, tangentMode=RCTM_User, time=1, value=1}]
       }}}
     Scale Curve=&RandomRangeFloat(Minimum=0.5, Maximum=0.7)
    )
    Module.Index0Float0Write=true
    Module.Index0Float1Write=true
    Module.Index0Float2Write=true
    Module.Index0Float3Write=true
   )
   &SolveForcesAndVelocity(Clamp Velocity=false, Limit Acceleration=false)
  ]
  rendererClasses [&NiagaraMeshRendererProperties, &NiagaraSpriteRendererProperties]
  renderers [
   {rendererClass=&NiagaraMeshRendererProperties, rendererIndex=0}
   {rendererClass=&NiagaraMeshRendererProperties, rendererIndex=1}
   {rendererClass=&NiagaraMeshRendererProperties, rendererIndex=2}
   {rendererClass=&NiagaraSpriteRendererProperties, rendererIndex=3}
  ]
 }
 NE_VC_DriftTrail_WheelWrap_Parent {
  bEnabled=false
  emitterSpawnScript [/SetVariables(
    Emitter.ContrastTx1:NiagaraFloat=&SelectFloatFromArray(
     Direct Array Index=$System.ColorVariantID
     Float Selection Array=DataInterface() {propertyValues {FloatData [0,0,0.8,1,0.75,0.75,1,0.95,1,1,0.97,1,1,0.95]}}
    )
    Emitter.Desaturation:NiagaraFloat=&SelectFloatFromArray(Direct Array Index=$System.ColorVariantID, Float Selection Array=DataInterface() {propertyValues {FloatData [1,0,0,0,0,0,0,0,0,0,0,0,0,0]}})
   ) {isSetParameters=true, module=SetVariables_#4}]
  emitterUpdateScript [
   &EmitterState(Inactive Response=::NewEnumerator1, Life Cycle Mode=::NewEnumerator1, Loop Behavior=::NewEnumerator1, Loop Duration=1, UseLoopDelay=false)
   &SpawnBurst_Instantaneous(Spawn Count=2, Use Spawn Probability=false, UseLoopCountLimit=false)
  ]
  particleSpawnScript [
   &VC_PaintedVariantsOverride(
    Black=<rgba 0.2 0.2 0.2 4>
    BurntSienna=<rgba 0.68 0.3373 0.2992 4>
    Cobalt=<rgba 0 0.075 0.9 4>
    Crimson=<rgba 1 0.1 0.1244 4>
    ForestGreen=<rgba 0.125 0.5 0.1439 4>
    Grey=<rgba 0.432 0.4387 0.48 4>
    Lime=<rgba 0.533 0.82 0 4>
    Orange=<rgba 1.12 0.28 0 4>
    Pink=<rgba 1.1 0.275 0.9625 4>
    Purple=<rgba 0.4934 0.28 0.8 4>
    Saffron=<rgba 1.1 1.0688 0.165 4>
    SkyBlue=<rgba 0 0.74 1.2 4>
    TitaniumWhite=<rgba 1.1 1.1 1.1 4>
   )
   &InitializeParticle(
    Color Mode=::NewEnumerator0
    Lifetime=$Emitter.CurrentLoopDuration
    Mesh Scale Mode=::NewEnumerator4
    Sprite Rotation Mode=::NewEnumerator2
    Sprite Size Mode=::NewEnumerator3
    UsePositionOffset=false
    Write Material Random=true
   )
   &PartitionParticles()
   &SkeletalMeshLocation(
    Bone / Socket Index=&Add_Integer(A=$Particles.PartitionParticles.Partition, B=4)
    Bone / Socket Sampling Mode=::NewEnumerator5
    Interpolate Mesh Sampling=false
    MaskSpawnGroup=false
    Mesh Sampling Type=::NewEnumerator6
    Skeletal Mesh=$System.SkeletalMesh
   )
   &InitialMeshOrientation(
    Mesh Orientation Mode=::NewEnumerator7
    Rotation=&SelectVectorFromArray(
     Direct Array Index=$Particles.PartitionParticles.Partition
     Vector Selection Array=DataInterface() {propertyValues {FloatData [<0 0 0>, <-0.5 0 0>]}}
    )
    Use Rotation Vector=true
   ) {module=InitialMeshOrientation001}
   /SetVariables(
    Particles.Vector:Vector3f=&SelectVectorFromArray(
     Direct Array Index=$Particles.PartitionParticles.Partition
     Vector Selection Array=DataInterface() {propertyValues {FloatData [<0 0 0>, <0.95 0.7 0.95>]}}
    )
    Particles.WheelSize:Vector3f=&SelectVectorFromArray(
     Direct Array Index=$Particles.PartitionParticles.Partition
     Vector Selection Array=DataInterface() {propertyValues {FloatData [<0.95 1 0.95>, <0 0.7 0>]}}
    )
   ) {isSetParameters=true}
   &InitialMeshOrientation(Mesh Orientation Mode=::NewEnumerator7, Rotation=<0 0 0.5>, Use Rotation Vector=true) {enabled=false}
   &Color(Color=<rgba 0 1 0 1>) {enabled=false}
  ]
  particleUpdateScript [
   &FN_DisableTireFX(
    Mask Index Attribute:NiagaraInt32=$Particles.PartitionParticles.Partition
    Mask Index FL=2
    Mask Index FR=3
    Mask Index RL=0
    Mask Index RR=1
   )
   &VC_WheelWrapScale(Wheel Index Attribute:NiagaraInt32=&Add_Integer(A=$Particles.PartitionParticles.Partition, B=2))
   &ParticleState(Kill Particles When Lifetime Has Elapsed=false)
   &ScaleMeshSize(Scale Factor=&SelectVectorFromArray(
     Direct Array Index=$Particles.Initial.PartitionParticles.Partition
     Vector Selection Array=DataInterface() {propertyValues {FloatData [<0 0 0>, <1 1 1>]}}
    )) {enabled=false}
   &UpdateMeshOrientation(Rotation Coordinate Space=::NewEnumerator0, Rotation Vector=<0 2 0>) {enabled=false}
   &SkeletalMeshLocation(
    Bone / Socket Index=&Add_Integer(A=$Particles.PartitionParticles.Partition, B=4)
    Bone / Socket Sampling Mode=::NewEnumerator5
    MaskSpawnGroup=false
    Mesh Sampling Type=::NewEnumerator6
    Orientation Sampling=::NewEnumerator1
    Sampled Position Offset=&SelectVectorFromArray(
     Direct Array Index=$Particles.Initial.PartitionParticles.Partition
     Vector Selection Array=DataInterface() {propertyValues {FloatData [<0 0 0>, <0 0 0>]}}
    )
    Skeletal Mesh=$System.SkeletalMesh
   ) {module=SkeletalMeshLocation001}
   &SolveForcesAndVelocity(Clamp Velocity=false, Limit Acceleration=false)
   &ScaleColor(
    Scale RGB=&VectorFromFloat(Value=&Waveform("Waveform [1]"=::NewEnumerator0, "[1] Phase Scale"=5, "[1] Frequency"=0.5, "[1] Amplitude Min/Max"=<50 100>))
    ScaleA=true
    ScaleRGB=true
   ) {enabled=false}
  ]
  rendererClasses [&NiagaraMeshRendererProperties]
  renderers [
   {rendererClass=&NiagaraMeshRendererProperties, rendererIndex=0}
   {rendererClass=&NiagaraMeshRendererProperties, rendererIndex=1}
  ]
 }
 NE_VC_DriftTrail_WheelWrap_Parent001 {
  emitterSpawnScript [/SetVariables(
    Emitter.ContrastTx1:NiagaraFloat=&SelectFloatFromArray(
     Direct Array Index=$System.ColorVariantID
     Float Selection Array=DataInterface() {propertyValues {FloatData [0,0,1.25,1,0.75,0.75,1.25,0.95,1,1,0.97,1.25,1.25,0.95]}}
    )
    Emitter.Desaturation:NiagaraFloat=&SelectFloatFromArray(Direct Array Index=$System.ColorVariantID, Float Selection Array=DataInterface() {propertyValues {FloatData [1,0,0,0,0,0,0,0,0,0,0,0,0,0]}})
   ) {isSetParameters=true, module=SetVariables_#4}]
  emitterUpdateScript [
   &EmitterState(
    Inactive Response=::NewEnumerator1
    Life Cycle Mode=::NewEnumerator1
    Loop Behavior=::NewEnumerator1
    Loop Duration Mode=::NewEnumerator1
    UseLoopDelay=false
   )
   &SpawnBurst_Instantaneous(Spawn Count=2, Use Spawn Probability=false, UseLoopCountLimit=false)
  ]
  particleSpawnScript [
   &VC_PaintedVariantsOverride(
    Black=<rgba 0.2 0.2 0.2 8>
    BurntSienna=<rgba 0.68 0.3373 0.2992 4>
    Cobalt=<rgba 0 0.075 0.9 4>
    Crimson=<rgba 1 0.1 0.1244 4>
    ForestGreen=<rgba 0.125 0.5 0.1439 8>
    Grey=<rgba 0.432 0.4387 0.48 4>
    Lime=<rgba 0.533 0.82 0 4>
    Orange=<rgba 1.12 0.28 0 4>
    Pink=<rgba 1.1 0.275 0.9625 4>
    Purple=<rgba 0.4934 0.28 0.8 8>
    Saffron=<rgba 1.1 1.0688 0.165 8>
    SkyBlue=<rgba 0 0.74 1.2 4>
    TitaniumWhite=<rgba 1.1 1.1 1.1 4>
   )
   &InitializeParticle(
    Color Mode=::NewEnumerator0
    Lifetime=5
    Mesh Scale Mode=::NewEnumerator4
    Sprite Rotation Mode=::NewEnumerator2
    Sprite Size Mode=::NewEnumerator3
    UsePositionOffset=false
    Write Material Random=true
   )
   &PartitionParticles()
   &SkeletalMeshLocation(
    Bone / Socket Index=&Add_Integer(A=$Particles.PartitionParticles.Partition, B=4)
    Bone / Socket Sampling Mode=::NewEnumerator5
    Interpolate Mesh Sampling=false
    MaskSpawnGroup=false
    Mesh Sampling Type=::NewEnumerator6
    Skeletal Mesh=$System.SkeletalMesh
   )
   &InitialMeshOrientation(
    Mesh Orientation Mode=::NewEnumerator7
    Rotation=&SelectVectorFromArray(
     Direct Array Index=$Particles.PartitionParticles.Partition
     Vector Selection Array=DataInterface() {propertyValues {FloatData [<0 0 0>, <-0.5 0 0>]}}
    )
    Use Rotation Vector=true
   )
   &Color(Color=<rgba 0 1 0 1>) {enabled=false}
  ]
  particleUpdateScript [
   &FN_DisableTireFX(
    Mask Index Attribute:NiagaraInt32=$Particles.PartitionParticles.Partition
    Mask Index FL=2
    Mask Index FR=3
    Mask Index RL=0
    Mask Index RR=1
   )
   &VC_WheelWrapScale(Wheel Index Attribute:NiagaraInt32=&Add_Integer(A=$Particles.PartitionParticles.Partition, B=2))
   &ParticleState(Kill Particles When Lifetime Has Elapsed=false)
   &ScaleMeshSize(Scale Factor=&SelectVectorFromArray(
     Direct Array Index=$Particles.Initial.PartitionParticles.Partition
     Vector Selection Array=DataInterface() {propertyValues {FloatData [<0 0 0>, <1 1 1>]}}
    )) {enabled=false}
   &UpdateMeshOrientation(Rotation Coordinate Space=::NewEnumerator0, Rotation Vector=<0 2 0>)
   &SkeletalMeshLocation(
    Bone / Socket Index=&Add_Integer(A=$Particles.PartitionParticles.Partition, B=4)
    Bone / Socket Sampling Mode=::NewEnumerator5
    MaskSpawnGroup=false
    Mesh Sampling Type=::NewEnumerator6
    Orientation Sampling=::NewEnumerator1
    Sampled Position Offset=&SelectVectorFromArray(
     Direct Array Index=$Particles.Initial.PartitionParticles.Partition
     Vector Selection Array=DataInterface() {propertyValues {FloatData [<0 0 0>, <0 0 0>]}}
    )
    Skeletal Mesh=$System.SkeletalMesh
   ) {module=SkeletalMeshLocation001}
   &SolveForcesAndVelocity(Clamp Velocity=false, Limit Acceleration=false)
   &ScaleColor(
    Scale RGB=&VectorFromFloat(Value=&Waveform("Waveform [1]"=::NewEnumerator0, "[1] Phase Scale"=5, "[1] Frequency"=0.5, "[1] Amplitude Min/Max"=<50 100>))
    ScaleA=true
    ScaleRGB=true
   ) {enabled=false}
  ]
  rendererClasses [&NiagaraMeshRendererProperties]
  renderers [
   {rendererClass=&NiagaraMeshRendererProperties, rendererIndex=0}
   {rendererClass=&NiagaraMeshRendererProperties, rendererIndex=1}
  ]
 }
}
summary {
 systemName=NS_DriftTrail_IconColors
 emitters [
  {emitterName=NE_VC_DriftTrail_Ribbon_Parent, rendererClasses [&NiagaraRibbonRendererProperties]}
  {bEnabled=false, emitterName=NE_VC_DriftTrail_WheelWrap_Parent, rendererClasses [&NiagaraMeshRendererProperties]}
  {
   emitterName=NE_VC_DriftTrail_Sprite_Parent
   rendererClasses [&NiagaraMeshRendererProperties, &NiagaraSpriteRendererProperties]
  }
  {emitterName=NE_VC_DriftTrail_WheelWrap_Parent001, rendererClasses [&NiagaraMeshRendererProperties]}
 ]
 userVariables [
  User.bIsFrontendPreview {type=NiagaraBool}
  User.bIsPainted {type=NiagaraBool}
  User.blsIsCoroColor {type=NiagaraBool}
  User.bPreviewVehicleSpeed {type=NiagaraBool}
  User.ColorVariantID {type=NiagaraInt32, defaultValue=2}
  User.GameModeID {type=NiagaraInt32}
  User.offset {type=Vector3f, description="offset from wheels(set through cosmetic template)", defaultValue {struct=Vector3f, value=<0 0 0>}}
  User.PaintedColor {type=LinearColor, description="", defaultValue {struct=LinearColor, value=<rgba 1 1 1 1>}}
  User.PaintedScalar {type=NiagaraFloat, defaultValue=1, description=""}
  User.PortCount {type=NiagaraInt32}
  User.VehicleFlags {type=NiagaraInt32, defaultValue=806289408}
  User.Wheel_Radius_Array {
   type=NiagaraDataInterfaceArrayFloat
   defaultValue {
    struct=NiagaraExt_VariableValue_DataInterface
    value {
     dataInterface=&self:NiagaraDataInterfaceArrayFloat_1
     dataInterfaceClass=&NiagaraDataInterfaceArrayFloat
    }
   }
  }
  User.Wheel_Width_Array {
   type=NiagaraDataInterfaceArrayFloat
   defaultValue {
    struct=NiagaraExt_VariableValue_DataInterface
    value {
     dataInterface=&self:NiagaraDataInterfaceArrayFloat_0
     dataInterfaceClass=&NiagaraDataInterfaceArrayFloat
    }
   }
  }
  User.WheelTouchingArray {
   type=NiagaraDataInterfaceArrayFloat
   defaultValue {
    struct=NiagaraExt_VariableValue_DataInterface
    value {
     dataInterface=&self:NiagaraDataInterfaceArrayFloat_2
     dataInterfaceClass=&NiagaraDataInterfaceArrayFloat
    }
   }
  }
 ]
}
